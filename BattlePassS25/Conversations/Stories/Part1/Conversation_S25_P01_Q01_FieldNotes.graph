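# Conversation database for the Season 25 Part 1 quest 01 explorer field notes: five readable notes about the search for the Chalice of Cubetastrophe.
# Each entry tag opens a numbered title line offering Read Notes or Steal Notes.
# Reading shows the note text, then a Back node; stealing carries a per-note quest choice tag and ends the branch.
# Steal branches display "[Notes Stolen]", except note 4, which shows note-like text instead.
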ConversationDatabase /BattlePassS25/Conversations/Stories/Part1/Conversation_S25_P01_Q01_FieldNotes
entry_points [
 {tag=Conversation.Entry.Season25.FieldNotes.01, destinations [#1]}
 {tag=Conversation.Entry.Season25.FieldNotes.02, destinations [#2]}
 {tag=Conversation.Entry.Season25.FieldNotes.03, destinations [#3]}
 {tag=Conversation.Entry.Season25.FieldNotes.04, destinations [#4]}
 {tag=Conversation.Entry.Season25.FieldNotes.05, destinations [#5]}
]
nodes [
 ConversationEntryPointNode_14 {
  id=#5
  class=ConversationEntryPointNode
  outputs [#6]
  properties {entryTag {tagName=Conversation.Entry.Season25.FieldNotes.05}}
 }
 Speech_C_51 {
  id=#6
  class=Speech_C
  outputs [#7, #8]
  properties {generalConfig {defaultMessage="Explorer Field Notes ##5"}}
 }
 Speech_C_53 {
  id=#8
  class=Speech_C
  outputs []
  properties {generalConfig {defaultMessage="[Notes Stolen]"}}
  sub_nodes [ConversationChoiceNode_46 {
    class=ConversationChoiceNode
    properties {defaultChoiceDisplayText=Steal Notes, choiceTags {gameplayTags [{tagName=Athena.Quests.S25.Fieldnotes.Convo.Choice.05}]}}
   }]
 }
 Speech_C_52 {
  id=#7
  class=Speech_C
  outputs [#9]
  properties {generalConfig {defaultMessage="...can be contained, if we can find it in time..."}}
  sub_nodes [ConversationChoiceNode_47 {class=ConversationChoiceNode, properties {defaultChoiceDisplayText=Read Notes}}]
 }
 Back_C_8 {id=#9, class=Back_C}
 ConversationEntryPointNode_13 {
  id=#4
  class=ConversationEntryPointNode
  outputs [#10]
  properties {entryTag {tagName=Conversation.Entry.Season25.FieldNotes.04}}
 }
 Speech_C_50 {
  id=#10
  class=Speech_C
  outputs [#11, #12]
  properties {generalConfig {defaultMessage="Explorer Field Notes ##4"}}
 }
 Speech_C_48 {
  id=#12
  class=Speech_C
  outputs []
  properties {generalConfig {defaultMessage=We have to find the cup. But how?}}
  sub_nodes [ConversationChoiceNode_49 {
    class=ConversationChoiceNode
    properties {defaultChoiceDisplayText=Steal Notes, choiceTags {gameplayTags [{tagName=Athena.Quests.S25.Fieldnotes.Convo.Choice.04}]}}
   }]
 }
 Speech_C_49 {
  id=#11
  class=Speech_C
  outputs [#13]
  properties {generalConfig {defaultMessage=...find the cup. But how?}}
  sub_nodes [ConversationChoiceNode_48 {class=ConversationChoiceNode, properties {defaultChoiceDisplayText=Read Notes}}]
 }
 Back_C_7 {id=#13, class=Back_C}
 ConversationEntryPointNode_9 {
  id=#3
  class=ConversationEntryPointNode
  outputs [#14]
  properties {entryTag {tagName=Conversation.Entry.Season25.FieldNotes.03}}
 }
 Speech_C_35 {
  id=#14
  class=Speech_C
  outputs [#15, #16]
  properties {generalConfig {defaultMessage="Explorer Field Notes ##3"}}
 }
 Speech_C_37 {
  id=#16
  class=Speech_C
  outputs []
  properties {generalConfig {defaultMessage="[Notes Stolen]"}}
  sub_nodes [ConversationChoiceNode_38 {
    class=ConversationChoiceNode
    properties {defaultChoiceDisplayText=Steal Notes, choiceTags {gameplayTags [{tagName=Athena.Quests.S25.Fieldnotes.Convo.Choice.03}]}}
   }]
 }
 Speech_C_36 {
  id=#15
  class=Speech_C
  outputs [#17]
  properties {generalConfig {defaultMessage=...found a new wielder and is gaining power...}}
  sub_nodes [ConversationChoiceNode_37 {class=ConversationChoiceNode, properties {defaultChoiceDisplayText=Read Notes}}]
 }
 Back_C_3 {id=#17, class=Back_C}
 ConversationEntryPointNode_8 {
  id=#2
  class=ConversationEntryPointNode
  outputs [#18]
  properties {entryTag {tagName=Conversation.Entry.Season25.FieldNotes.02}}
 }
 Speech_C_32 {
  id=#18
  class=Speech_C
  outputs [#19, #20]
  properties {generalConfig {defaultMessage="Explorer Field Notes ##2"}}
 }
 Speech_C_34 {
  id=#20
  class=Speech_C
  outputs []
  properties {generalConfig {defaultMessage="[Notes Stolen]"}}
  sub_nodes [ConversationChoiceNode_36 {
    class=ConversationChoiceNode
    properties {defaultChoiceDisplayText=Steal Notes, choiceTags {gameplayTags [{tagName=Athena.Quests.S25.Fieldnotes.Convo.Choice.02}]}}
   }]
 }
 Speech_C_33 {
  id=#19
  class=Speech_C
  outputs [#21]
  properties {generalConfig {defaultMessage="...a powerful item made by the cubes, capable of leveling the entire Island..."}}
  sub_nodes [ConversationChoiceNode_35 {class=ConversationChoiceNode, properties {defaultChoiceDisplayText=Read Notes}}]
 }
 Back_C_2 {id=#21, class=Back_C}
 ConversationEntryPointNode_2 {
  id=#1
  class=ConversationEntryPointNode
  outputs [#22]
  properties {entryTag {tagName=Conversation.Entry.Season25.FieldNotes.01}}
 }
 Speech_C_29 {
  id=#22
  class=Speech_C
  outputs [#23, #24]
  properties {generalConfig {defaultMessage="Explorer Field Notes ##1"}}
 }
 Speech_C_31 {
  id=#24
  class=Speech_C
  outputs []
  properties {generalConfig {defaultMessage="[Notes Stolen]"}}
  sub_nodes [ConversationChoiceNode_34 {
    class=ConversationChoiceNode
    properties {defaultChoiceDisplayText=Steal Notes, choiceTags {gameplayTags [{tagName=Athena.Quests.S25.Fieldnotes.Convo.Choice.01}]}}
   }]
 }
 Speech_C_30 {
  id=#23
  class=Speech_C
  outputs [#25]
  properties {generalConfig {defaultMessage=Reports confirm the CHALICE OF CUBETASTROPHE is in this rainforest.}}
  sub_nodes [ConversationChoiceNode_33 {class=ConversationChoiceNode, properties {defaultChoiceDisplayText=Read Notes}}]
 }
 Back_C_1 {id=#25, class=Back_C}
]
