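# Conversation database for a Season 25 story quest where the player examines petroglyphs and a map leading to the Chalice.
# Five tagged entry points each open a two-speech chain, with the second speech gated behind an examine choice.
# The four petroglyph choices carry per-glyph quest gameplay tags as the copied glyphs progressively reveal a map.
# The map choice has no tag and places the Chalice at the former Anvil Square.
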ConversationDatabase /BattlePassS25/Conversations/Stories/Part1/Conversation_S25_P01_Q02_Petroglyphs
entry_points [
 {tag=Conversation.Entry.Season25.Petroglyph.01, destinations [#1]}
 {tag=Conversation.Entry.Season25.Petroglyph.02, destinations [#2]}
 {tag=Conversation.Entry.Season25.Petroglyph.03, destinations [#3]}
 {tag=Conversation.Entry.Season25.Petroglyph.04, destinations [#4]}
 {tag=Conversation.Entry.Season25.Petroglyph.Map, destinations [#5]}
]
nodes [
 ConversationEntryPointNode_3 {
  id=#5
  class=ConversationEntryPointNode
  outputs [#6]
  properties {entryTag {tagName=Conversation.Entry.Season25.Petroglyph.Map}}
 }
 Speech_C_12 {
  id=#6
  class=Speech_C
  outputs [#7]
  properties {generalConfig {defaultMessage=It's a map to the Chalice's location.}}
 }
 Speech_C_11 {
  id=#7
  class=Speech_C
  outputs []
  properties {generalConfig {defaultMessage=The Chalice appears to be found where Anvil Square once stood.}}
  sub_nodes [ConversationChoiceNode_6 {class=ConversationChoiceNode, properties {defaultChoiceDisplayText=Examine Map Closely}}]
 }
 ConversationEntryPointNode_13 {
  id=#4
  class=ConversationEntryPointNode
  outputs [#8]
  properties {entryTag {tagName=Conversation.Entry.Season25.Petroglyph.04}}
 }
 Speech_C_50 {
  id=#8
  class=Speech_C
  outputs [#9]
  properties {generalConfig {defaultMessage=A series of glyphs are carved into the rock...}}
 }
 Speech_C_48 {
  id=#9
  class=Speech_C
  outputs []
  properties {generalConfig {defaultMessage="[You copy the final part of the petroglyph. It's a map to the chalice!]"}}
  sub_nodes [ConversationChoiceNode_3 {
    class=ConversationChoiceNode
    properties {defaultChoiceDisplayText=Examine, choiceTags {gameplayTags [{tagName=Athena.Quests.S25.Petroglyph.Convo.Choice.04}]}}
   }]
 }
 ConversationEntryPointNode_9 {
  id=#3
  class=ConversationEntryPointNode
  outputs [#10]
  properties {entryTag {tagName=Conversation.Entry.Season25.Petroglyph.03}}
 }
 Speech_C_5 {
  id=#10
  class=Speech_C
  outputs [#11]
  properties {generalConfig {defaultMessage=A series of glyphs are carved into the rock...}}
 }
 Speech_C_6 {
  id=#11
  class=Speech_C
  outputs []
  properties {generalConfig {defaultMessage="[You copy the petroglyph. Some of these glyphs look like locations on the Island...]"}}
  sub_nodes [ConversationChoiceNode_1 {
    class=ConversationChoiceNode
    properties {defaultChoiceDisplayText=Examine, choiceTags {gameplayTags [{tagName=Athena.Quests.S25.Petroglyph.Convo.Choice.03}]}}
   }]
 }
 ConversationEntryPointNode_8 {
  id=#2
  class=ConversationEntryPointNode
  outputs [#12]
  properties {entryTag {tagName=Conversation.Entry.Season25.Petroglyph.02}}
 }
 Speech_C_4 {
  id=#12
  class=Speech_C
  outputs [#13]
  properties {generalConfig {defaultMessage=A series of glyphs are carved into the rock...}}
 }
 Speech_C_3 {
  id=#13
  class=Speech_C
  outputs []
  properties {generalConfig {defaultMessage="[As you copy the petroglyph, you notice a glyph resembling a waterfall.]"}}
  sub_nodes [ConversationChoiceNode_2 {
    class=ConversationChoiceNode
    properties {defaultChoiceDisplayText=Examine, choiceTags {gameplayTags [{tagName=Athena.Quests.S25.Petroglyph.Convo.Choice.02}]}}
   }]
 }
 ConversationEntryPointNode_2 {
  id=#1
  class=ConversationEntryPointNode
  outputs [#14]
  properties {entryTag {tagName=Conversation.Entry.Season25.Petroglyph.01}}
 }
 Speech_C_0 {
  id=#14
  class=Speech_C
  outputs [#15]
  properties {generalConfig {defaultMessage=A series of glyphs are carved into the rock...}}
 }
 Speech_C_1 {
  id=#15
  class=Speech_C
  outputs []
  properties {generalConfig {defaultMessage="[You carefully copy the petroglyph's design into your notes.]"}}
  sub_nodes [ConversationChoiceNode_49 {
    class=ConversationChoiceNode
    properties {defaultChoiceDisplayText=Examine, choiceTags {gameplayTags [{tagName=Athena.Quests.S25.Petroglyph.Convo.Choice.01}]}}
   }]
 }
]
