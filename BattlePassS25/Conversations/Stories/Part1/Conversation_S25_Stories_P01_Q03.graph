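# Conversation database for the Season 25 Stories Part 1 Quest 3 talk with Aura, who points the player to a map to the chalice and suspects Orelia.
# A single entry point leads into a linear chain of three speech nodes.
# The auto-selected main-questline choice requires Aura's entry tag on the NPC, the quest active and its first objective not yet completed.
# The final choice carries the Aura Q03 choice tag.
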ConversationDatabase /BattlePassS25/Conversations/Stories/Part1/Conversation_S25_Stories_P01_Q03
entry_points [{tag=Conversation.Entry.Season25.P01.Q03.Aura, destinations [#1]}]
nodes [
 ConversationEntryPointNode_0 {
  id=#1
  class=ConversationEntryPointNode
  outputs [#2]
  properties {entryTag {tagName=Conversation.Entry.Season25.P01.Q03.Aura}}
 }
 Speech_C_3 {
  id=#2
  class=Speech_C
  outputs [#3]
  properties {
   softTaskIcon=&T_UI_NPC_MainQuest
   generalConfig {defaultMessage=I created a map to the chalice based on your notes. It's up to you to find it.}
  }
  sub_nodes [
   ConversationChoiceNode_1 {
    class=ConversationChoiceNode
    properties {
     defaultChoiceDisplayText=Report to Aura
     choiceTags {gameplayTags [{tagName=Conversation.Choice.AutoSelectIfValid}, {tagName=Conversation.Choice.Category.MainQuestline}]}
    }
   }
   FortConversationRequirement_ControllerRequirement_3 {
    class=FortConversationRequirement_ControllerRequirement
    properties {
     nodeName=Require: NPC Has Conversation Tag
     participantId {tagName=Conversation.Participant.NPC}
     requirement {class=FortControllerRequirement_NPCHasConversationEntryTag, properties {queryTag {tagName=Conversation.Entry.Aura}}}
    }
   }
   FortConversationRequirement_ControllerRequirement_0 {
    class=FortConversationRequirement_ControllerRequirement
    properties {
     nodeName=Require: Is Quest Active
     requirement {
      class=FortControllerRequirement_IsQuestActive
      properties {questDefinition=/BattlePassS25/Items/QuestItems/Stories/Part01/quest_s25_stories_p01_q03_01}
     }
    }
   }
   FortConversationRequirement_ControllerRequirement_2 {
    class=FortConversationRequirement_ControllerRequirement
    properties {
     nodeName=Require: Has Completed Quest Objective
     requirement {
      class=FortControllerRequirement_HasCompletedQuestObjective
      properties {
       bInverseCondition=true
       objectiveBackendName=quest_s25_stories_p01_q03_01_obj0
       questDefinition=&quest_s25_stories_p01_q03_01
      }
     }
    }
   }
  ]
 }
 Speech_C_28 {
  id=#3
  class=Speech_C
  outputs [#4]
  properties {generalConfig {defaultMessage=This feels like Orelia's handiwork. She likes messing with gold objects.}}
  sub_nodes [ConversationChoiceNode_3 {class=ConversationChoiceNode, properties {defaultChoiceDisplayText=Continue}}]
 }
 Speech_C_0 {
  id=#4
  class=Speech_C
  outputs []
  properties {generalConfig {defaultMessage="Maybe bring some Bars with you, just in case you need to lure her out."}}
  sub_nodes [ConversationChoiceNode_0 {
    class=ConversationChoiceNode
    properties {defaultChoiceDisplayText=Continue, choiceTags {gameplayTags [{tagName=Athena.Quests.S25.Conversations.Aura.Choice.Q03}]}}
   }]
 }
]
