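# Conversation database for Season 25 Stories Part 1 Quest 4, where the player hands the Chalice of Cubetastrophe to Aura.
# One entry tag leads into a linear three-line speech chain with Talk, Hand Over and Continue choices.
# The auto-selected main-questline opener requires the Aura NPC tag, the active q04_02 quest, objective 0 complete and objective 1 incomplete.
# The final Continue choice carries an Aura Q04 quest tag.
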ConversationDatabase /BattlePassS25/Conversations/Stories/Part1/Conversation_S25_Stories_P01_Q04
entry_points [{tag=Conversation.Entry.Season25.P01.Q04.Aura, destinations [#1]}]
nodes [
 ConversationEntryPointNode_0 {
  id=#1
  class=ConversationEntryPointNode
  outputs [#2]
  properties {entryTag {tagName=Conversation.Entry.Season25.P01.Q04.Aura}}
 }
 Speech_C_3 {
  id=#2
  class=Speech_C
  outputs [#3]
  properties {
   softTaskIcon=&T_UI_NPC_MainQuest
   generalConfig {defaultMessage="We'll get Orelia and Oro resettled in some new gold objects. Maybe a fork and knife? "}
  }
  sub_nodes [
   ConversationChoiceNode_1 {
    class=ConversationChoiceNode
    properties {
     defaultChoiceDisplayText=Talk to Aura
     choiceTags {gameplayTags [{tagName=Conversation.Choice.AutoSelectIfValid}, {tagName=Conversation.Choice.Category.MainQuestline}]}
    }
   }
   FortConversationRequirement_ControllerRequirement_3 {
    class=FortConversationRequirement_ControllerRequirement
    properties {
     nodeName=Require: NPC Has Conversation Tag
     participantId {tagName=Conversation.Participant.NPC}
     requirement {class=FortControllerRequirement_NPCHasConversationEntryTag, properties {queryTag {tagName=Conversation.Entry.Aura}}}
    }
   }
   FortConversationRequirement_ControllerRequirement_0 {
    class=FortConversationRequirement_ControllerRequirement
    properties {
     nodeName=Require: Is Quest Active
     requirement {
      class=FortControllerRequirement_IsQuestActive
      properties {questDefinition=/BattlePassS25/Items/QuestItems/Stories/Part01/quest_s25_stories_p01_q04_02}
     }
    }
   }
   FortConversationRequirement_ControllerRequirement_4 {
    class=FortConversationRequirement_ControllerRequirement
    properties {
     nodeName=Require: Has Completed Quest Objective
     requirement {
      class=FortControllerRequirement_HasCompletedQuestObjective
      properties {
       objectiveBackendName=quest_s25_stories_p01_q04_02_obj0
       questDefinition=&quest_s25_stories_p01_q04_02
      }
     }
    }
   }
   FortConversationRequirement_ControllerRequirement_2 {
    class=FortConversationRequirement_ControllerRequirement
    properties {
     nodeName=Require: Has Completed Quest Objective
     requirement {
      class=FortControllerRequirement_HasCompletedQuestObjective
      properties {
       bInverseCondition=true
       objectiveBackendName=quest_s25_stories_p01_q04_02_obj1
       questDefinition=&quest_s25_stories_p01_q04_02
      }
     }
    }
   }
  ]
 }
 Speech_C_28 {
  id=#3
  class=Speech_C
  outputs [#4]
  properties {generalConfig {defaultMessage="I'll get my contact to seal away the Chalice of Cubetastrophe. Without a wielder, it's powerless."}}
  sub_nodes [ConversationChoiceNode_0 {class=ConversationChoiceNode, properties {defaultChoiceDisplayText=Hand Over Chalice}}]
 }
 Speech_C_0 {
  id=#4
  class=Speech_C
  outputs []
  properties {generalConfig {defaultMessage="Thanks, Looper. I owe ya one."}}
  sub_nodes [ConversationChoiceNode_4 {
    class=ConversationChoiceNode
    properties {defaultChoiceDisplayText=Continue, choiceTags {gameplayTags [{tagName=Athena.Quests.S25.Conversations.Aura.Choice.Q04}]}}
   }]
 }
]
